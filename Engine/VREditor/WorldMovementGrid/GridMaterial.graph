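# Unlit translucent material for the VR Editor world movement grid, drawing a tiled grid texture that fades radially.
# A radial gradient feeds cosine and power curves that blend gray, orange and pink into a bright, bloomy color.
# Opacity combines the grid texture's alpha, a softer cosine falloff and OpacityParam.
# With depth testing disabled, a pixel-versus-scene depth comparison keeps occluded parts faintly visible instead of hiding them.
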Material /Engine/VREditor/WorldMovementGrid/GridMaterial
graph {
 properties {bDisableDepthTest=true, blendMode=BLEND_Translucent, pixelDepthOffsetMode=PDOM_Legacy, shadingModel=MSM_Unlit, usedShadingModels=Unlit}
 nodes [
  n1 = MaterialFunctionCall!(Radius, CenterPosition) {materialFunction=/Engine/Functions/Engine_MaterialFunctions01/Gradient/RadialGradient}
  ConstantBiasScale(None=@n1.RadialGradient) {bias=0, desc=Linear, scale=1}
  n2 = Constant3Vector() {constant=<rgba 0.175 0.175 0.175 0>}
  n3 = Constant3Vector() {constant=<rgba 0.91 0.494 0 0>}
  n4 = OneMinus(None=@n1.RadialGradient)
  n5 = Cosine(None=@n4) {desc=Cos Hard, period=4}
  n6 = Power(Base=@n5) {constExponent=5}
  Multiply(A=LinearInterpolate(A=LinearInterpolate(A=@n2, B=@n3, Alpha=@n6), B=Constant3Vector() {constant=<rgba 1 0 0.7285 0>}, Alpha=Power(Base=@n5) {constExponent=40})) {constB=6, desc=Bright and bloomy}
  n7 = TextureSample(UVs=Multiply(A=TextureCoordinate(), B=Constant() {r=35})) {texture=&/SimpleGrid}
  n8 = ConstantBiasScale(None=Cosine(None=@n4) {period=2}) {desc=Cos Soft}
  Multiply(
   A=Multiply(A=Multiply(A=@n7.A, B=@n8), B=ScalarParameter() {defaultValue=1, parameterName=OpacityParam})
   B=If(
    A=PixelDepth().R
    B=SceneDepth().R
    "A > B"=Constant() {desc=Opacity when occluded, r=0.05}
    "A < B"=Constant() {desc=Opacity when visible in plain sight, r=1}
   )
  )
 ]
}
