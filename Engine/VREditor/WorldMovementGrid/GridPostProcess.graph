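# Post-process material for the VR editor's world movement grid that fogs the scene beyond a radius around RoomOrigin.
# Custom HLSL rebuilds world position from scene depth; its distance to RoomOrigin feeds exponential falloff between start and end distances scaled by WorldScaleFactor.
# That falloff and Opacity blend scene color toward desaturated, FogColor-tinted color, with full fog color past SkyboxDistance.
# A feature-level switch substitutes constant white on ES2/ES3_1.
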Material /Engine/VREditor/WorldMovementGrid/GridPostProcess
graph {
 properties {
  blendableLocation=BL_SceneColorBeforeDOF
  materialDomain=MD_PostProcess
  pixelDepthOffsetMode=PDOM_Legacy
  shadingModel=MSM_Unlit
  usedShadingModels=Unlit
 }
 nodes [
  n1 = SceneTexture() {sceneTextureId=PPI_PostProcessInput0}
  n2 = VectorParameter() {desc=FogColor, parameterName=FogColor, defaultValue=<rgba 0.012 0.0125 0.012 1>}
  MaterialExpressionTextureCoordinate_1 = TextureCoordinate()
  MaterialExpressionSceneDepth_0 = SceneDepth(UVs=@MaterialExpressionTextureCoordinate_1)
  n3 = Custom!(ScreenPos=@MaterialExpressionTextureCoordinate_1, SceneDepth=@MaterialExpressionSceneDepth_0.R) {
   code="float4 Pos = mul(float4(((ScreenPos.xy - View.ScreenPositionScaleBias.wz ) / View.ScreenPositionScaleBias.xy) * SceneDepth, SceneDepth, 1.0), LWCToFloat(ResolvedView.ScreenToWorld));  return Pos.xyz * 1 / Pos.w;"
   desc="\n*** Original source before expression upgrade ***\nfloat4 Pos = mul(float4(((ScreenPos.xy - View.ScreenPositionScaleBias.wz ) / View.ScreenPositionScaleBias.xy) * SceneDepth, SceneDepth, 1.0), View.ScreenToWorld);  return Pos.xyz * 1 / Pos.w;"
   inputs [
    {inputName=ScreenPos, input {expression=&self:MaterialExpressionTextureCoordinate_1}}
    {inputName=SceneDepth, input {expression=&self:MaterialExpressionSceneDepth_0, mask=1, maskR=1}}
   ]
  }
  n4 = Distance(A=@n3, B=VectorParameter() {desc=(Set programmatically), parameterName=RoomOrigin}.RGB)
  n5 = Max(
   A=Subtract(A=@n4, B=ScalarParameter() {defaultValue=8000, desc=SkyboxDistance, parameterName=SkyboxDistance})
   B=ScalarParameter() {defaultValue=0.9, desc=FogOpacity, parameterName=FogOpacity}
  )
  n6 = ScalarParameter() {defaultValue=1, desc=(Set programmatically), parameterName=WorldScaleFactor}
  n7 = Multiply(A=ScalarParameter() {defaultValue=600, desc=StartDistance, parameterName=StartDistance}, B=@n6)
  n8 = Multiply(A=@n6, B=ScalarParameter() {defaultValue=640, desc=EndDistance, parameterName=EndDistance})
  n9 = MaterialFunctionCall!(Depth=Divide(A=Subtract(A=@n4, B=@n7), B=Subtract(A=@n8, B=@n7)), Density, Use Exp2) {materialFunction=/Engine/Functions/Engine_MaterialFunctions01/Density/ExponentialDensity}
  n10 = Power(Base=OneMinus(None=ScalarParameter() {defaultValue=1, desc=Opacity, parameterName=Opacity}))
  n11 = LinearInterpolate(
   A=@n1.Color
   B=LinearInterpolate(A=Desaturation(None=@n1.Color), B=AppendVector(A=@n2.RGB, B=@n2.RGB), Alpha=Clamp(None=@n5)) {constAlpha=0.95, constB=0.1}
   Alpha=Multiply(A=OneMinus(None=@n9.Result), B=OneMinus(None=@n10))
  )
  n12 = Constant4Vector() {constant=<rgba 1 1 1 0>}
  FeatureLevelSwitch(Default=@n11, ES2=@n12.RGBA, ES3_1=@n12.RGBA, SM4_REMOVED=@n11, SM5=@n11, SM6=@n11)
  Multiply() {constA=1}
  WorldPosition()
 ]
}
